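# Translucent unlit overlay for skeletal meshes that draws a flickering, colored soft outline, with stencil settings described for radioactive actors.
# A custom HLSL node radially samples custom stencil with depth occlusion, hardened on low-end platforms.
# Distance and flicker drive thickness, which pushes vertices along normals with optional sway.
# Color supports two tones and time-of-day intensity; panning noise, custom-depth or Fresnel culling, and eye-adaptation branches supply further shaping.
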Material /WildEstateTech/Gameplay/Materials/M_Buzzcool_Overlay
graph {
 properties {
  bAutomaticallySetUsageInEditor=false
  bUsedWithMeshDeformer=true
  bUsedWithSkeletalMesh=true
  blendMode=BLEND_Translucent
  shadingModel=MSM_Unlit
  usedShadingModels=Unlit
 }
 nodes [
  n1 = SceneTexture() {sceneTextureId=PPI_CustomDepth}
  n2 = ScalarParameter() {
   defaultValue=5
   desc=Stencil value for radioactive actors so the custom depth doesn't go away when backfaced by other custom depth actors
   group=WildEstate_CustomDepth
   parameterName=Stencil_Value
  }
  n3 = Multiply(
   A=SceneTexture() {sceneTextureId=PPI_SceneDepth}.Color
   B=MaterialFunctionCall!(ID=@n2) {materialFunction=/Game/Packages/Fortress_SharedMaterials/Stencil_ID_Definitions/StencilMaskCompare}.Result
  )
  n4 = Ceil(None=ComponentMask(None=Subtract(A=@n1.Color, B=@n3)) {r=true})
  n5 = TwoSidedSign()
  n6 = Round(None=Fresnel())
  n7 = StaticSwitchParameter(True=Multiply(A=Saturate(None=@n4), B=NamedRerouteUsage()), False=Multiply(A=OneMinus(None=@n5), B=OneMinus(None=@n6))) {desc=Skeletal actors may need this to help with interior culling, group=WildEstate_CustomDepth, parameterName=Use_CustomDepth}
  Clamp(None=@n7)
  Constant() {desc="This value used to be set to ZERO which mean\r\nEmissive Fun out would be Zero on low end\r\nchanging to 1 for now", r=1}
  n8 = Multiply(A=Time(), B=ScalarParameter() {defaultValue=4, group=WildEstate_Softoutline_Noise, parameterName=Speed noise, sortPriority=33})
  n9 = ScalarParameter() {defaultValue=1.5, group=WildEstate_Softoutline_Noise, parameterName=Texture FPS noise, sortPriority=33}
  Divide(A=Floor(None=Multiply(A=@n8, B=@n9)), B=@n9)
  n10 = MaterialFunctionCall!(
   Emissive=NamedRerouteUsage()
   Alpha=ScalarParameter() {defaultValue=1, group=Cosmetic Emissive EyeAdaptation, parameterName=Cosmetic Emissive EyeAdaptation Alpha}
   Material
  ) {materialFunction=/Game/Packages/Fortress_SharedMaterials/Base_Material_Functions/BaseCharacterShaderFunctions/MF_CosmeticEmissiveEyeAdaptation}
  LinearInterpolate(A=@n10.Emissive, B=Constant3Vector())
  n11 = Multiply(
   A=ScreenPosition().ViewportUV
   B=ScalarParameter() {defaultValue=2.4394, group=WildEstate_Softoutline_Noise, parameterName=tile noise, sortPriority=33}
  )
  n12 = Time()
  n13 = AppendVector(
   A=ScalarParameter() {defaultValue=0.1, group=WildEstate_Softoutline_Noise, parameterName=X speed noise, sortPriority=33}
   B=ScalarParameter() {defaultValue=0.3, group=WildEstate_Softoutline_Noise, parameterName=Y speed noise, sortPriority=33}
  )
  n14 = Multiply(
   A=TextureSampleParameter2D(UVs=Panner(Coordinate=@n11, Time=@n12, Speed=@n13) {bFractionalPart=true, speedX=0.02, speedY=1}) {group=WildEstate_Softoutline_Noise, parameterName=NoiseTexture, texture=/WildEstateTech/Gameplay/Textures/Water_01_Pack}.RGB
   B=ScalarParameter() {defaultValue=1, group=WildEstate_Softoutline_Noise, parameterName=Noise_Multi}
  )
  n15 = ScalarParameter() {defaultValue=1, group=WildEstate_Softoutline_Noise, parameterName=Noise_Pow}
  NamedRerouteDeclaration(None=Saturate(None=Power(Base=@n14, Exp=@n15))) {name=panning_noise, nodeColor=<rgba 0.0078 1 0 1>}
  n16 = NamedRerouteUsage()
  MaterialExpressionScreenPosition_1 = ScreenPosition()
  MaterialExpressionNamedRerouteUsage_12 = NamedRerouteUsage()
  MaterialExpressionScalarParameter_29 = ScalarParameter() {defaultValue=7, group=WildEstate_Softoutline, parameterName=DistanceSteps, sortPriority=31}
  MaterialExpressionScalarParameter_30 = ScalarParameter() {defaultValue=6, group=WildEstate_Softoutline, parameterName=RadialSteps, sortPriority=31}
  MaterialExpressionScalarParameter_31 = ScalarParameter() {defaultValue=0.62, group=WildEstate_Softoutline, parameterName=RadialOffset, sortPriority=31}
  MaterialExpressionScalarParameter_32 = ScalarParameter() {defaultValue=-1.2996, group=WildEstate_Softoutline, parameterName=KernelPower, sortPriority=31}
  MaterialExpressionScalarParameter_10 = ScalarParameter() {
   defaultValue=5
   desc=Stencil value for radioactive actors so the custom depth doesn't go away when backfaced by other custom depth actors
   group=WildEstate_CustomDepth
   parameterName=Stencil_Value
  }
  MaterialExpressionSceneTexture_4 = SceneTexture() {sceneTextureId=PPI_CustomStencil}
  n17 = Custom!(
   UV=@MaterialExpressionScreenPosition_1.ViewportUV
   Distance=@MaterialExpressionNamedRerouteUsage_12
   DistanceSteps=@MaterialExpressionScalarParameter_29
   RadialSteps=@MaterialExpressionScalarParameter_30
   RadialOffset=@MaterialExpressionScalarParameter_31
   KernelPower=@MaterialExpressionScalarParameter_32
   StencilIndex=@MaterialExpressionScalarParameter_10
   Dummy=@MaterialExpressionSceneTexture_4.Color
  ) {
   code="float3 CurColor = 0;\r\nfloat2 NewUV = UV;\r\nint i = 0;\r\nfloat StepSize = Distance / (int) DistanceSteps;\r\nfloat CurDistance = 0;\r\nfloat2 CurOffset = 0;\r\nfloat SubOffset = 0;\r\nfloat TwoPi = 6.283185;\r\nfloat accumdist = 0;\r\n\r\nfloat2 texelSize = View.ViewSizeAndInvSize.zw * Distance;\r\n//float tmpstencil = 0;\r\n//float tmpstencil = SceneTextureLookup(ViewportUVToBufferUV(UV), 25, false);\r\n\r\nif (DistanceSteps < 1)\r\n{\r\n    \r\n\treturn 0;\r\n}\r\n \r\nelse\r\n{\r\n\twhile (i < (int) DistanceSteps)\r\n\t{\r\n\t\tCurDistance += StepSize;\r\n\t\tfor (int j = 0; j < (int) RadialSteps; j++)\r\n\t\t{\r\n\t\t\tSubOffset++;\r\n\t\t\tCurOffset.x = cos(TwoPi * (SubOffset / RadialSteps)) * texelSize.r;\r\n\t\t\tCurOffset.y = sin(TwoPi * (SubOffset / RadialSteps)) * texelSize.g;\r\n\t\t\tNewUV.x = UV.x + CurOffset.x * CurDistance;\r\n\t\t\tNewUV.y = UV.y + CurOffset.y * CurDistance;\r\n\t\t\tfloat distpow = pow(CurDistance, KernelPower);\r\n            float tmpstencil = SceneTextureLookupCommon(Parameters, ViewportUVToBufferUV(NewUV), 25, false);\r\n            float validmask = 1-floor(saturate(abs(tmpstencil-StencilIndex)+0.01));\r\n            float tmpdepth = SceneTextureLookupCommon(Parameters, ViewportUVToBufferUV(NewUV), 13, false);\r\n            float tmpscene = CalcSceneDepth(ScreenAlignedPosition(GetScreenPosition(Parameters)))+3.00000f;\r\n            float occlud = select((tmpdepth.r > tmpscene), 0.00000000f, 1.00000000f);  //occlusion may be to expensive in a loop\r\n\t\t\tCurColor += SceneTextureLookupCommon(Parameters, ViewportUVToBufferUV(NewUV), PPI_CustomStencil, false) * distpow * validmask*occlud;\r\n\t\t\taccumdist += distpow;\r\n\t\t}\r\n\t\tSubOffset += RadialOffset;\r\n\t\ti++;\r\n\t}\r\n\treturn CurColor / accumdist;\r\n}\r\n"
   showCode=true
   inputs [
    {inputName=UV, input {expression=&self:MaterialExpressionScreenPosition_1}}
    {inputName=Distance, input {expression=&self:MaterialExpressionNamedRerouteUsage_12}}
    {inputName=DistanceSteps, input {expression=&self:MaterialExpressionScalarParameter_29}}
    {inputName=RadialSteps, input {expression=&self:MaterialExpressionScalarParameter_30}}
    {inputName=RadialOffset, input {expression=&self:MaterialExpressionScalarParameter_31}}
    {inputName=KernelPower, input {expression=&self:MaterialExpressionScalarParameter_32}}
    {inputName=StencilIndex, input {expression=&self:MaterialExpressionScalarParameter_10}}
    {
     inputName=Dummy
     input {
      expression=&self:MaterialExpressionSceneTexture_4
      mask=1
      maskA=1
      maskB=1
      maskG=1
      maskR=1
     }
    }
   ]
  }
  n18 = Saturate(None=ComponentMask(None=@n17) {r=true})
  n19 = Ceil(None=@n18)
  n20 = MaterialFunctionCall("Xb1,PS4"=@n18, "PC-High&&Epic,XSX,PS5"=@n18, "PC-Perf,Android"=@n19, Switch=@n19, "PC-Low(NotPerf)&&Medium"=@n18) {materialFunction=&PlatformSwitch_1D}
  NamedRerouteDeclaration(None=StaticSwitchParameter(True=@n16, False=Multiply(A=@n16, B=@n20.Result)) {group=WildEstate_CustomDepth, parameterName=Quantize_Color}) {name=emissive_function, nodeColor=<rgba 0.5937 1 0 1>}
  n21 = NamedRerouteUsage()
  n22 = ObjectOrientation()
  n23 = MaterialFunctionCall(Speed=ScalarParameter() {defaultValue=0.1, group=WPO Sway, parameterName=RotationSpeedX}) {materialFunction=&DebugTimeSine}
  n24 = ScalarParameter() {defaultValue=0.1, group=WPO Sway, parameterName=RotationSpeedZ}
  n25 = MaterialFunctionCall(Float3=VectorParameter() {group=WPO Sway, parameterName=RotationAxis, defaultValue=<rgba 1 1 0 1>}.RGB) {materialFunction=&BreakOutFloat3Components}
  n26 = MaterialFunctionCall(Speed=ScalarParameter() {defaultValue=0.15, group=WPO Sway, parameterName=RotationSpeedY}) {materialFunction=&DebugTimeSine}
  n27 = MaterialFunctionCall(X=Multiply(A=Add(A=@n23.Result, B=@n24), B=@n25.R), Y=Multiply(A=Add(A=@n26.Result, B=@n24), B=@n25.G), Z=@n25.B) {materialFunction=&MakeFloat3}
  n28 = VectorParameter() {group=WPO Sway, parameterName=PivotOffset, defaultValue=<rgba 0 0 0 1>}
  n29 = Time()
  n30 = MaterialFunctionCall(Float3=ObjectPositionWS()) {materialFunction=&BreakOutFloat3Components}
  n31 = Add(A=@n30.R, B=@n30.R)
  n32 = Add(A=@n29, B=Divide(A=Add(A=@n31, B=@n30.B), B=ScalarParameter() {defaultValue=0.1, group=WPO Sway, parameterName=TimeOffsetAmount}))
  n33 = Sine(None=Multiply(
    A=StaticSwitchParameter(True=@n32, False=@n29) {defaultValue=true, group=WPO Sway, parameterName=TimeLocationOffset}
    B=ScalarParameter() {defaultValue=0.5, group=WPO Sway, parameterName=SwaySpeed}
   ))
  n34 = ConstantBiasScale(None=@n33)
  n35 = ScalarParameter() {defaultValue=0.005, group=WPO Sway, parameterName=SwayAmountMin}
  n36 = ScalarParameter() {defaultValue=0.02, group=WPO Sway, parameterName=SwayAmountMax}
  n37 = ConstantBiasScale(None=Sine(None=@n32) {period=6})
  n38 = RotateAboutAxis(
   NormalizedRotationAxis=Normalize(VectorInput=MaterialFunctionCall!(VectorToRotate=@n22, LookAtVector=@n27.Result, RestingVector) {materialFunction=/Engine/Functions/Engine_MaterialFunctions01/Vectors/RotateVector}.Result)
   RotationAngle=Multiply(A=StaticSwitchParameter(True=@n34, False=@n33) {group=WPO Sway, parameterName=BiasSineWave}, B=LinearInterpolate(A=@n35, B=@n36, Alpha=@n37))
   PivotPoint=Add(
    A=MaterialFunctionCall() {materialFunction=&ObjectPivotPoint}.Object Pivot Location
    B=Transform(None=@n28.RGB) {transformSourceType=TRANSFORMSOURCE_Local}
   )
   Position=WorldPosition().XYZ
  )
  n39 = Constant3Vector() {constant=<rgba 0 0 1 0>}
  n40 = Multiply(A=Time(), B=ScalarParameter() {defaultValue=0.015, group=WPO Sway, parameterName=ConstantRotate_Speed, sortPriority=310})
  n41 = ObjectPositionWS()
  n42 = WorldPosition()
  NamedRerouteDeclaration(None=Add(A=Add(A=@n21, B=@n38), B=RotateAboutAxis(NormalizedRotationAxis=@n39, RotationAngle=@n40, PivotPoint=@n41, Position=@n42.XYZ))) {name=NewWPO, nodeColor=<rgba 0.4297 1 0 1>}
  n43 = VectorParameter() {group=WPO Sway, parameterName=WPO_Movement_Max, sortPriority=300, defaultValue=<rgba 180 180 180 1>}
  n44 = MaterialFunctionCall(Color=Divide(A=ObjectPositionWS(), B=ScalarParameter() {defaultValue=64, group=WPO Sway, parameterName=PositionOffset, sortPriority=300})) {materialFunction=&SplitComponents}
  n45 = VectorParameter() {group=WPO Sway, parameterName=WPO_Speed, sortPriority=300, defaultValue=<rgba 0.04 0.0175 0.03 2>}
  n46 = Add(A=Frac(None=Multiply(A=@n44.R, B=@n44.R)), B=Multiply(A=Time(), B=Multiply(A=@n45.RGB, B=@n45.RGB)))
  NamedRerouteDeclaration(None=LinearInterpolate(
    A=VectorParameter() {group=WPO Sway, parameterName=WPO_Movement_Min, sortPriority=300, defaultValue=<rgba -1 1 0 1>}.RGB
    B=@n43.RGB
    Alpha=Sine(None=@n46)
   )) {name=Bob, nodeColor=<rgba 0 0.7656 1 1>}
  NamedRerouteDeclaration(None=@n38) {name=WPOFFinal, nodeColor=<rgba 0 1 0.0625 1>}
  NamedRerouteDeclaration(None=@n20.Result) {name=MaskMath, nodeColor=<rgba 1 0 0.1875 1>}
  n47 = LinearInterpolate(
   A=ScalarParameter() {defaultValue=1.5, group=WildEstate_Softoutline, parameterName=Thickness_Far, sortPriority=31}
   B=ScalarParameter() {defaultValue=0.5, group=WildEstate_Softoutline, parameterName=Thickness, sortPriority=31}
   Alpha=NamedRerouteUsage()
  )
  n48 = LinearInterpolate(
   A=@n47
   B=Add(A=@n47, B=ScalarParameter() {defaultValue=1, group=WildEstate_Softoutline, parameterName=Thickness_Flicker, sortPriority=31})
   Alpha=NamedRerouteUsage()
  )
  n49 = Subtract(
   A=@n47
   B=LinearInterpolate(
    A=ScalarParameter() {group=WildEstate_Softoutline, parameterName=mobile_thickness_reduction_far}
    B=ScalarParameter() {group=WildEstate_Softoutline, parameterName=mobile_thickness_reduction}
    Alpha=NamedRerouteUsage()
   ) {constA=0.5}
  )
  n50 = MaterialFunctionCall("Xb1,PS4"=@n48, "PC-High&&Epic,XSX,PS5"=@n48, "PC-Perf,Android"=@n49, Switch=@n49, "PC-Low(NotPerf)&&Medium"=@n48) {materialFunction=&PlatformSwitch_1D}
  n51 = ScalarParameter() {defaultValue=1, group=WildEstate_Softoutline, parameterName=Thickness offset_Far, sortPriority=31}
  n52 = ScalarParameter() {defaultValue=8, group=WildEstate_Softoutline, parameterName=thickness offset, sortPriority=31}
  n53 = NamedRerouteUsage()
  n54 = LinearInterpolate(A=@n50.Result, B=Add(A=@n50.Result, B=LinearInterpolate(A=@n51, B=@n52, Alpha=@n53)), Alpha=NamedRerouteUsage())
  NamedRerouteDeclaration(None=@n54) {name=Thickness, nodeColor=<rgba 0 1 0.9297 1>}
  n55 = VectorParameter() {group=WildEstate_Softoutline, parameterName=Color, defaultValue=<rgba 0.4164 2 0.0873 1>}
  n56 = VectorParameter() {group=WildEstate_Softoutline, parameterName=ColorB, defaultValue=<rgba 0 2 0.5019 1>}
  n57 = NamedRerouteUsage()
  n58 = StaticSwitchParameter(
   True=LinearInterpolate(A=@n55.RGB, B=@n56.RGB, Alpha=ScalarParameter() {parameterName=ColorBlendAlpha})
   False=StaticSwitchParameter(True=LinearInterpolate(A=@n56.RGB, B=@n55.RGB, Alpha=@n57), False=@n55.RGB) {group=WildEstate_Softoutline, parameterName=MultiTone_Color}
  ) {parameterName=UseBlendableColors}
  n59 = LinearInterpolate(
   A=@n58
   B=Multiply(A=@n58, B=ScalarParameter() {defaultValue=2, group=WildEstate_Softoutline, parameterName=Color_Flicker_Str})
   Alpha=NamedRerouteUsage()
  )
  n60 = MaterialFunctionCall(
   Value_Morning=ScalarParameter() {defaultValue=1, group=TOD, parameterName="Value_Morning (S)"}
   Value_Day=ScalarParameter() {defaultValue=1, group=TOD, parameterName="Value_Day (S)"}
   Value_Evening=ScalarParameter() {defaultValue=1, group=TOD, parameterName="Value_Evening (S)"}
   Value_Night=ScalarParameter() {defaultValue=1, group=TOD, parameterName="Value_Night (S)"}
  ) {materialFunction=&Scalar_Time_of_Day}
  NamedRerouteDeclaration(None=StaticSwitchParameter(True=Multiply(A=@n59, B=@n60.Result), False=@n59) {defaultValue=true, parameterName=UseTOD_Intensity?}) {name=Color, nodeColor=<rgba 1 0.6094 0 1>}
  n61 = ScalarParameter() {group=WildEstate_Softoutline_Flicker, parameterName=FlickerAdd}
  n62 = Time()
  n63 = Multiply(A=ScalarParameter() {defaultValue=1, group=WildEstate_Softoutline_Flicker, parameterName=FlickerSpeed}, B=@n62)
  n64 = AppendVector(A=AppendVector(A=Sine(None=@n63) {period=1.6}, B=Sine(None=@n63) {period=1.42}), B=Sine(None=@n63) {period=1.51})
  n65 = ScalarParameter() {defaultValue=1, group=WildEstate_Softoutline_Flicker, parameterName=FlickerFrequency}
  n66 = Add(A=Sine(None=@n63) {period=2.1}, B=Multiply(B=@n65) {constA=0.25})
  n67 = Add(A=Sine(None=@n63) {period=0.72}, B=Multiply(B=@n65) {constA=0.75}) {constB=0.75}
  n68 = Ceil(None=Add(A=Multiply(B=@n65) {constA=0.9, constB=0.9}, B=Sine(None=@n63) {period=1.25}) {constB=0.9})
  n69 = Saturate(None=Multiply(A=Multiply(A=@n66, B=@n67), B=Multiply(A=@n67, B=@n68)))
  n70 = LinearInterpolate(Alpha=ConstantBiasScale(None=Sine(None=@n62) {period=0.075})) {constA=0.7}
  n71 = Multiply(
   A=Multiply(A=Ceil(None=@n64), B=Multiply(A=@n69, B=@n70))
   B=ScalarParameter() {defaultValue=1, group=WildEstate_Softoutline_Flicker, parameterName=FlickerMult}
  )
  NamedRerouteDeclaration(None=Saturate(None=Add(A=@n61, B=@n71))) {name=flicker_Speed, nodeColor=<rgba 1 0.5391 0 1>}
  n72 = Length(None=ActorPositionWS() {originType=CameraRelative}) {is=class MaterialExpressionLength}
  n73 = ScalarParameter() {defaultValue=1000, group=WildEstate_Softoutline, parameterName=Distance}
  n74 = Power(Base=Divide(A=Subtract(A=@n72, B=@n73) {constB=100}) {constB=400}) {constExponent=0.5}
  NamedRerouteDeclaration(None=OneMinus(None=Saturate(None=@n74))) {name=distance blending, nodeColor=<rgba 0.5938 0 1 1>}
  Reroute(None=@n7)
  n75 = NamedRerouteUsage()
  n76 = Multiply(A=VertexNormalWS(), B=@n54)
  StaticSwitchParameter(True=Constant(), False=StaticSwitchParameter(True=Add(A=@n75, B=@n76), False=@n76) {group=WPO Sway, parameterName=UseSwayWPO}) {desc=USed for ZP overlay, parameterName=ZP_overlay}
  TextureCoordinate()
 ]
}
